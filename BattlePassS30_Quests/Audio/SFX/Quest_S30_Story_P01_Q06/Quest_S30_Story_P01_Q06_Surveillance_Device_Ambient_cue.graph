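# Looping ambient Sound Cue for the surveillance device in a Season 30 Battle Pass story quest.
# A single looping wave player feeds a Modulator that applies slight random variation, with its upper pitch limit set to unity.
# The cue plays in the quests sound class with a volume boost and borrows the supply drop HiddenStar ambient loop attenuation for distance falloff.
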SoundCue /BattlePassS30_Quests/Audio/SFX/Quest_S30_Story_P01_Q06/Quest_S30_Story_P01_Q06_Surveillance_Device_Ambient_cue
root=Modulator_0
nodes [
 {id=Modulator_0, children [WavePlayer_5], properties {pitchMax=1}}
 {
  id=WavePlayer_5
  properties {bLooping=true, soundWaveAssetPtr=&/Quest_S30_Story_P01_Q06_Surveillance_Device_Amb_01}
 }
]
properties {
 attenuationSettings=/Game/Athena/Sounds/SupplyDrops/Fort_HiddenStar_Ambient_Loop_Att
 maxDistance=2500
 soundClassObject=&Fort_Quests
 volumeMultiplier=1.25
}
