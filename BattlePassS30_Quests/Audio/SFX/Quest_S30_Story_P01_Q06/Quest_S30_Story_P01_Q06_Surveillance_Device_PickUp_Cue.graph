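# Sound cue for picking up the surveillance device in a Season 30 Battle Pass story quest.
# A random node picks one of two equally weighted pickup waves and feeds a modulator with pitch variation flattened.
# A mixer layers that, boosted in volume, over a reused reward-screen chest upgrade UI sound.
# The cue uses the quests sound class and borrows attenuation settings from a Hidden Star ambient loop.
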SoundCue /BattlePassS30_Quests/Audio/SFX/Quest_S30_Story_P01_Q06/Quest_S30_Story_P01_Q06_Surveillance_Device_PickUp_Cue
root=Mixer_0
nodes [
 {id=Mixer_0, children [Modulator_0,WavePlayer_4], properties {inputVolume [3,1]}}
 {id=Modulator_0, children [Random_0], properties {pitchMax=1, pitchMin=1}}
 {id=Random_0, children [WavePlayer_0,WavePlayer_1], properties {weights [1,1]}}
 {
  id=WavePlayer_0
  properties {soundWaveAssetPtr=&/Quest_S30_Story_P01_Q06_Surveillance_Device_Pickup_01}
 }
 {
  id=WavePlayer_1
  properties {soundWaveAssetPtr=&/Quest_S30_Story_P01_Q06_Surveillance_Device_Pickup_02}
 }
 {id=WavePlayer_4, properties {soundWaveAssetPtr=/Game/Sounds/Fort_UI_Sounds/Rewards/Fort_RewardScreen_TeamScore_ChestLootUpgrade}}
]
properties {
 attenuationSettings=/Game/Athena/Sounds/SupplyDrops/Fort_HiddenStar_Ambient_Loop_Att
 maxDistance=2500
 soundClassObject=&Fort_Quests
}
